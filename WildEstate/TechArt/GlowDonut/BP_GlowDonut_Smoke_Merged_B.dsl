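; GlowDonut smoke actor Blueprint that can recolor its materials to an "Apocolypse" palette.
; When UseApocolypseColor is set, BeginPlay loops over a ten-entry component array and swaps each material slot for a dynamic instance.
; Each instance receives ApocolypseColor and ApocolypseColor2 through its Color and Color2 parameters.
; The Change Color custom event duplicates this logic; overlap, tick and construction are empty.
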(event Custom|Change%20Color
  (Variables|Default|GetUseApocolypseColor
    (:then
      (for _array_element (Utilities|Array|MakeArray 0 0 0 0 0 0 0 0 0 0)
        (for (_array_element_2 _array_index) (Rendering|Material|GetMaterials _array_element)
          (bind _returnvalue (Rendering|Material|CreateDynamicMaterialInstance _array_element_2 "None" "None" :!class /Script/Engine.KismetMaterialLibrary))
          (Rendering|Material|SetMaterial _array_element _array_index _returnvalue)
          (Rendering|Material|SetVectorParameterValue _returnvalue "Color" (Variables|Default|GetApocolypseColor) :!class /Script/Engine.MaterialInstanceDynamic)
          (Rendering|Material|SetVectorParameterValue _returnvalue "Color2" (Variables|Default|GetApocolypseColor2) :!class /Script/Engine.MaterialInstanceDynamic))))
    (:else)))

(event EventBeginPlay
  (Variables|Default|GetUseApocolypseColor
    (:then
      (for _array_element (Utilities|Array|MakeArray 0 0 0 0 0 0 0 0 0 0)
        (for (_array_element_2 _array_index) (Rendering|Material|GetMaterials _array_element)
          (bind _returnvalue (Rendering|Material|CreateDynamicMaterialInstance _array_element_2 "None" "None" :!class /Script/Engine.KismetMaterialLibrary))
          (Rendering|Material|SetMaterial _array_element _array_index _returnvalue)
          (Rendering|Material|SetVectorParameterValue _returnvalue "Color" (Variables|Default|GetApocolypseColor) :!class /Script/Engine.MaterialInstanceDynamic)
          (Rendering|Material|SetVectorParameterValue _returnvalue "Color2" (Variables|Default|GetApocolypseColor2) :!class /Script/Engine.MaterialInstanceDynamic))))
    (:else)))

(event Collision|EventActorBeginOverlap
  (params (OtherActor :type Actor%20Object%20Reference)))

(event EventTick
  (params (DeltaSeconds :type Float%20%28single-precision%29)))

(fn ConstructionScript)
