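# Unlit, masked material for Niagara mesh particles that fakes toon lighting with a hard shadow line.
# Sky atmosphere light direction dotted with a normal-map-perturbed vertex normal is contrast-sharpened to blend dynamic-parameter shadow color into particle color.
# The result is tinted by Color Texture, optionally lerped toward dimmed texture, overlaid with masked outline color, then passed through MF_FX_EmissiveAdjustments.
# Normal tiling and offset use only X channels for both axes.
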Material /WildEstate/TechArt/GlowDonut/FX/M_ToonParticleMesh
graph {
 properties {
  bAutomaticallySetUsageInEditor=false
  bUsedWithNiagaraMeshParticles=true
  blendMode=BLEND_Masked
  ditherOpacityMask=true
  shadingModel=MSM_Unlit
  usedShadingModels=Unlit
 }
 nodes [
  DynamicParameter() {parameterIndex=1, paramNames [NA,NA,NA,NA]}
  DynamicParameter() {parameterIndex=2, paramNames [NA,NA,NA,NA]}
  DynamicParameter() {parameterIndex=3, paramNames [NA,NA,NA,NA]}
  LinearInterpolate() {constA=-0.5, constAlpha=0, constB=0.5}
  n1 = TextureSampleParameter2D() {parameterName=Color Texture, texture=/Engine/EngineResources/WhiteSquareTexture}
  n2 = ParticleColor()
  n3 = SkyAtmosphereLightDirection()
  n4 = VectorParameter() {parameterName=Normal Tile Offset, channelNames {a=Offset Y, b=Offset X, g=Tile Y, r=Tile X}, defaultValue=<rgba 1 1 0 0>}
  n5 = Add(A=Multiply(A=TextureCoordinate(), B=AppendVector(A=@n4.Tile X, B=@n4.Tile X)), B=AppendVector(A=@n4.Offset X, B=@n4.Offset X))
  n6 = LinearInterpolate(
   A=VertexNormalWS()
   B=Transform(None=TextureSample(UVs=@n5) {samplerType=SAMPLERTYPE_Normal, texture=/FortWater/Materials/WaterSurface/Mobile/Textures/T_MobileWater_LakeNormal}.RGB)
   Alpha=ScalarParameter() {parameterName=Normal Strength}
  )
  n7 = ConstantBiasScale(None=DotProduct(A=Normalize(VectorInput=@n3), B=Normalize(VectorInput=@n6)))
  n8 = VectorParameter() {parameterName=ShadowLine Options, channelNames {a=contrast, b=Offset, g=mul, r=pow}, defaultValue=<rgba 1 1 0 100>}
  n9 = Add(A=Multiply(A=Power(Base=@n7, Exp=@n8.pow), B=@n8.mul), B=@n8.Offset)
  n10 = Multiply(
   A=@n1.RGB
   B=LinearInterpolate(
    A=DynamicParameter() {paramNames [Shadow Color R, Shadow Color G, Shadow Color B, NA]}.RGB
    B=@n2.RGB
    Alpha=MaterialFunctionCall(In=@n9, Contrast=@n8.contrast) {materialFunction=&CheapContrast}.Result
   )
  )
  n11 = Multiply(A=@n1.RGB) {constB=0.2}
  n12 = ScalarParameter() {parameterName=Lerp Emissive}
  n13 = ChannelMaskParameter(None=TextureSampleParameter2D() {parameterName="Outline Mask (R)", texture=&Black}.RGBA) {parameterName=OutlinMaskChannel}
  n14 = VectorParameter() {parameterName=Outline Mask Options, channelNames {a=contrast, b=Offset, g=mul, r=pow}, defaultValue=<rgba 1 1 0 100>}
  n15 = Add(A=Multiply(A=Power(Base=@n13.R, Exp=@n14.pow), B=@n14.mul), B=@n14.Offset)
  MaterialFunctionCall(Emissive=LinearInterpolate(
    A=LinearInterpolate(A=@n10, B=@n11, Alpha=@n12)
    B=VectorParameter() {parameterName=Outline Color}.RGB
    Alpha=MaterialFunctionCall(In=@n15, Contrast=@n14.contrast) {materialFunction=&CheapContrast}.Result
   )) {materialFunction=&MF_FX_EmissiveAdjustments}
  MaterialFunctionCall() {materialFunction=&DebugTimeSine}
  Multiply()
  PixelNormalWS()
  Reroute(None=@n2.A)
  Subtract(A=@n8.Offset) {constB=0.5}
  VectorParameter() {parameterName=Surface for Lit, defaultValue=<rgba 0 0 1 0>}
  VectorParameter() {parameterName=Base Color, defaultValue=<rgba 0.1198 0.1198 0.1198 0>}
 ]
}
