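# Translucent, unlit, two-sided wind material for the WildEstate low-gravity effect, drawing wispy smoke streaks on particles.
# A scrolling flow map distorts smoke textures, tinted by two colors, particle color and time-of-day brightness.
# Particle random offsets UVs and picks and flips flipbook shapes; inverse Fresnel, depth, camera-distance and travel-direction fades shape opacity, minus a particle-driven dissolve.
# A world-position-offset branch supports hiding, scaling and shifting the mesh.
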Material /WildEstate/TechArt/GlowDonut/FX/M_WildEstate_LowGrav_Wind2
graph {
 properties {
  bUsedWithNanite=true
  bUsedWithNiagaraSprites=true
  blendMode=BLEND_Translucent
  pixelDepthOffsetMode=PDOM_Legacy
  shadingModel=MSM_Unlit
  twoSided=true
  usedShadingModels=Unlit
 }
 nodes [
  n1 = Subtract(A=ObjectPositionWS(), B=WorldPosition().XYZ)
  n2 = ScalarParameter() {bUseCustomPrimitiveData=true, parameterName=WPO_HideMesh Primitive}
  n3 = Constant()
  n4 = Constant() {r=1}
  n5 = StaticSwitchParameter(
   True=Multiply(B=If(A=@n2, "A > B"=@n3, "A == B"=@n4, "A < B"=@n3) {constB=3}) {constA=1}
   False=StaticSwitchParameter(
    True=ScalarParameter() {bUseCustomPrimitiveData=true, parameterName=WPO_HideMesh Primitive}
    False=ScalarParameter() {parameterName=WPO_HideMesh}
   ) {parameterName="Use Custom Prim Data (Chests)"}
  ) {parameterName="Use Custom Prim Data (Train/S28)"}
  n6 = MaterialFunctionCall() {materialFunction=&ObjectPivotPoint}
  n7 = WorldPosition()
  n8 = Multiply(
   A=Transform(None=Subtract(A=@n6.Object Pivot Location, B=@n7.XYZ)) {transformSourceType=TRANSFORMSOURCE_World, transformType=TRANSFORM_Local}
   B=OneMinus(None=VectorParameter() {parameterName=Scaled_MeshSizeXYZ, defaultValue=<rgba 1 1 1 1>}.RGB)
  )
  Add(
   A=Add(A=Multiply(A=@n1, B=@n5), B=Transform(None=@n8) {transformSourceType=TRANSFORMSOURCE_Local})
   B=Transform(None=VectorParameter() {parameterName=ObjectLocationOffset, defaultValue=<rgba 0 0 0 1>}.RGB) {transformSourceType=TRANSFORMSOURCE_Local}
  )
  n9 = VectorParameter() {parameterName=ToD, defaultValue=<rgba 0.7 1 0.7 0.7>}
  n10 = TextureObjectParameter() {
   group=MainTexture
   parameterName=MainTexture_Texture
   samplerType=SAMPLERTYPE_LinearColor
   sortPriority=0
   texture=/Game/Effects/Fort_Effects/Gameplay/SmokeDustExplosions/Textures/T_RibbonSmokeMasked_03
  }
  n11 = VectorParameter() {group=MainTexture, parameterName=MainTexture_UVControl, channelNames {a=VPan, b=UPan, g=VScale, r=UScale}, defaultValue=<rgba 1 1 1 0>}
  n12 = Multiply(A=Constant2Vector() {g=1}, B=ParticleRandom())
  n13 = TextureCoordinate()
  n14 = Add(A=@n12, B=@n13)
  n15 = VectorParameter() {parameterName=Swirl UV, defaultValue=<rgba 0.3 0.15 0.5 0.3>}
  n16 = ComponentMask(None=@n15.RGBA) {a=true, b=true}
  n17 = Add(A=Add(A=@n12, B=Multiply(A=@n13, B=@n16) {constB=0.5}), B=Multiply(A=Time(), B=ComponentMask(None=@n15.RGB) {g=true, r=true}))
  n18 = ComponentMask(None=TextureSample(UVs=MaterialFunctionCall(XY=@n17) {materialFunction=&Swizzle}.YX) {samplerType=SAMPLERTYPE_LinearColor, texture=/Game/Effects/Fort_Effects/Creatures/Elemental/Fire/MergedFireFlowMap}.RGB) {g=true, r=true}
  n19 = Add(
   A=MaterialFunctionCall(
    Texture=@n10
    UVControl_RGB=@n11
    UVControl_A=@n11
    "[Override]UVInput"=MaterialFunctionCall(XY=@n14) {materialFunction=&Swizzle}.YX
   ) {materialFunction=&FXCommon_TextureSample}.UV
   B=Multiply(A=Subtract(A=@n18) {constB=0.5}, B=ScalarParameter() {defaultValue=0.5, parameterName=Swirl Strength})
  )
  n20 = Saturate(None=Power(Base=TextureSample(UVs=@n19, Tex=@n10).R, Exp=ScalarParameter() {defaultValue=1}))
  n21 = LinearInterpolate(
   A=VectorParameter() {parameterName="BaseColor ", defaultValue=<rgba 1 1 1 1>}.RGB
   B=VectorParameter() {parameterName=SecColor, defaultValue=<rgba 1 0.7136 0.8319 1>}.RGB
   Alpha=Saturate(None=@n20)
  )
  n22 = ParticleColor()
  MaterialFunctionCall(Emissive=Multiply(
    A=MaterialFunctionCall(Value_Morning=@n9.R, Value_Day=@n9.R, Value_Evening=@n9.R, Value_Night=@n9.R) {materialFunction=&Scalar_Time_of_Day}.Result
    B=Multiply(A=@n21, B=@n22.RGB)
   )) {materialFunction=&MF_FX_EmissiveAdjustments}
  n23 = Subtract(A=Multiply(A=@n18) {constB=0.5}) {constB=0.25}
  n24 = TextureCoordinate()
  n25 = MaterialFunctionCall(XYZ=@n24, XY=@n24) {materialFunction=&Swizzle}
  n26 = Multiply(A=Time(), B=Constant2Vector() {r=0.74})
  n27 = TextureSample(UVs=Add(A=@n23, B=Add(A=@n25.YX, B=@n26))) {texture=/Game/Effects/Fort_Effects/Gameplay/SmokeDustExplosions/Textures/T_RibbonSmokeMasked_04}
  NamedRerouteDeclaration(None=Multiply(A=@n27.R, B=DynamicParameter() {defaultValue=<rgba 0 1 1 1>, paramNames [Dissolve,Param2,Param3,Param4]}.Dissolve)) {name=Dissolve, nodeColor=<rgba 0 1 0.7891 1>}
  n28 = VectorParameter() {parameterName=fresn, defaultValue=<rgba 1 3 3 1>}
  n29 = OneMinus(None=Fresnel(ExponentIn=@n28.R) {baseReflectFraction=0})
  n30 = Power(Base=Saturate(None=Multiply(A=@n29, B=@n28.B)), Exp=@n28.G)
  n31 = ParticleRandom()
  n32 = Multiply(A=@n31) {constB=4}
  n33 = Constant() {r=2}
  n34 = TextureCoordinate()
  n35 = Sign(None=Subtract(A=@n31) {constB=0.5})
  n36 = MaterialFunctionCall(
   "Animation  Phase (0-1)"=Divide(A=Floor(None=@n32)) {constB=4}
   Number of Rows=@n33
   Number of Columns=@n33
   Texture=TextureObject() {samplerType=SAMPLERTYPE_LinearColor, texture=&/T_WildEstate_Smoke_Trail}
   UVs=Add(A=@n23, B=Multiply(A=@n34, B=@n35))
  ) {materialFunction=&FlipBook}
  n37 = DepthFade(
   Opacity=Multiply(A=Multiply(A=@n20, B=@n27.G), B=ComponentMask(None=@n36.Result) {r=true})
   FadeDistance=ScalarParameter() {defaultValue=250, parameterName=DepthFadeDistance}
  )
  n38 = MaterialFunctionCall(Distance=ScalarParameter() {defaultValue=700, parameterName=Near Camera Fade}) {materialFunction=&Near_Camera_Fade}
  n39 = ParticleDirection()
  n40 = CameraVectorWS()
  n41 = OneMinus(None=Abs(None=DotProduct(A=@n39, B=@n40)))
  n42 = Multiply(A=Multiply(A=Multiply(A=@n30, B=@n37), B=Multiply(A=@n38.Result, B=@n38.Result)), B=Saturate(None=Add(A=@n41) {constB=0.333}))
  Subtract(A=Multiply(A=Saturate(None=@n42), B=ParticleColor().A), B=NamedRerouteUsage())
 ]
}
