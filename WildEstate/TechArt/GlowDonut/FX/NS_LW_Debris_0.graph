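# NS_LW_Debris_0 is a Niagara effect that, judging by its names, provides ambient wind-blown debris near the ground.
# It has two emitters, Debris_H and GroundWind, whose modules and renderers are not listed here.
# User parameters expose spawn rate, a wide, shallow spawn box range and a velocity defaulting to motion along negative Y.
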NiagaraSystem /WildEstate/TechArt/GlowDonut/FX/NS_LW_Debris_0
dependencies {usedDataInterfaces [], usedDynamicInputs [], usedModules [&SystemState]}
emitters {Debris_H {rendererClasses [], renderers []}, GroundWind {rendererClasses [], renderers []}}
summary {
 systemName=NS_LW_Debris_0
 emitters [{emitterName=Debris_H, rendererClasses []}, {emitterName=GroundWind, rendererClasses []}]
 userVariables [
  User.Spawn Rate {type=NiagaraFloat, defaultValue=2, description=""}
  User.SpawnBoxRange {type=Vector3f, description="", defaultValue {struct=Vector3f, value=<2000 2000 50>}}
  User.Velocity {type=Vector3f, description="", defaultValue {struct=Vector3f, value=<0 -500 0>}}
 ]
}
