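# Niagara debris suction effect that pulls sprites and meshes inward and upward over a wide area, amid rising ash.
# Fountain loops bursts on a ring with negative velocity, using spawn groups to split renderers and forces, with one burst's probability set per platform.
# GPU Ash emitter fills a large sphere, splitting particles into three partitions that switch color, size, lift, noise and renderer.
# User.Spawn gates ash spawning, camera fade scales ash size, and both emitters use distance culling.
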NiagaraSystem /WildEstate/TechArt/GlowDonut/FX/NS_LW_Debris_suction
dependencies {
 usedDataInterfaces [
  &NiagaraDataInterfacePlatformSet
  &NiagaraDataInterfaceCurve
  &NiagaraDataInterfaceVector2DCurve
  &NiagaraDataInterfaceSpriteRendererInfo
  &NiagaraDataInterfaceCamera
 ]
 usedDynamicInputs [&ReturnExecIndex]
 usedModules [&SystemState]
}
emitters {
 Ash {
  simTarget=GPUComputeSim
  emitterUpdateScript [
   &EmitterState(Enable Distance Culling=true, Module.bUseMaxDistance=true, Module.bUseMinDistance=false, Scalability Mode=::NewEnumerator1)
   &SpawnRate(Spawn Probability=$User.Spawn, SpawnRate=80, Use Spawn Probability=true)
  ]
  particleSpawnScript [
   &PartitionParticles(Number Of Partitions=3)
   /SetVariables(Particles.VisibilityTag=$Particles.PartitionParticles.Partition) {isSetParameters=true}
   &InitializeParticle(
    Color=&MergeHelper_ColorSwitch(
     0=<rgba 5 4.9603 4.5699 1>
     1=<rgba 0.0104 0.2865 0.0355 1>
     2=<rgba 0.0881 0.099 0.0902 1>
     3=?
     Index=$Particles.PartitionParticles.Partition
     Num=::NewEnumerator1
    )
    Lifetime=4
    Sprite Rotation Mode=::NewEnumerator2
    Sprite Size Mode=::NewEnumerator3
    Sprite UV Mode=::NewEnumerator3
    Uniform Sprite Size=&MergeHelper_FloatSwitch(
     0=&RandomRangeFloat(Minimum=20, Maximum=25)
     1=&RandomRangeFloat(Minimum=35, Maximum=60)
     2=&RandomRangeFloat(Minimum=8, Maximum=11)
     3=?
     Index=$Particles.PartitionParticles.Partition
     Num=::NewEnumerator1
    )
    UsePositionOffset=false
    Write Material Random=true
   )
   &ShapeLocation(
    MaskSpawnGroup=false
    Non Uniform Scale=&MergeHelper_VectorSwitch(0=<1.1 1.1 0.8>, 1=<1.1 1.1 0.8>, 2=<1.1 1.1 0.2>, 3=?, Index=$Particles.PartitionParticles.Partition, Num=::NewEnumerator1)
    Offset=&MergeHelper_VectorSwitch(0=<0 0 -350>, 1=<0 0 -350>, 2=<0 0 -500>, 3=?, Index=$Particles.PartitionParticles.Partition, Num=::NewEnumerator1)
    Offset Coordinate Space=::World
    Offset Mode=::NewEnumerator0
    Override Seed=false
    Sphere Radius=2250
    UseOwnerScale=true
   )
   &SubUVAnimation(
    Evaluation Type=::NewEnumerator0
    "Sprite Renderer "=DataInterface() {propertyValues {SpriteRenderer=&self:Ash_0.NiagaraSpriteRendererProperties_0}}
    SubUV Animation Mode:ENiagara_SubUVLookupModeV2=::NewEnumerator2
    UseEndFrame=false
    UseStartFrame=false
   )
   &SubUVAnimation(
    Evaluation Type=::NewEnumerator0
    "Sprite Renderer "=DataInterface() {propertyValues {SpriteRenderer=&self:Ash_0.NiagaraSpriteRendererProperties_2}}
    SubUV Animation Mode:ENiagara_SubUVLookupModeV2=::NewEnumerator2
    UseEndFrame=false
    UseStartFrame=false
   ) {module=SubUVAnimation002}
   &DynamicMaterialParameters(Module.Index0Float0Write=true, Module.Index0Float1Write=true, Module.Index0Float2Write=true, Module.Index0Float3Write=true) {enabled=false}
  ]
  particleUpdateScript [
   &ParticleState()
   &ScaleSpriteSize(
    Uniform Curve Scale=/Niagara/DynamicInputs/Camera/CameraFade(
     Camera Query:NiagaraDataInterfaceCamera=DataInterface()
     Fade Offset:NiagaraFloat=100
     Fade Distance:NiagaraFloat=200
     Use Owner Position:NiagaraBool=true
    )
    Uniform Curve Sprite Scale=DataInterface() {propertyValues {Curve {keys [
        {interpMode=RCIM_Cubic, tangentMode=RCTM_Break, time=0, value=0}
        {arriveTangent=2, interpMode=RCIM_Cubic, leaveTangent=2, time=0.1776, value=0.7441}
        {interpMode=RCIM_Cubic, tangentMode=RCTM_Break, time=0.5, value=1}
        {arriveTangent=-2, interpMode=RCIM_Cubic, leaveTangent=-2, time=0.7941, value=0.7912}
        {interpMode=RCIM_Cubic, tangentMode=RCTM_Break, time=1, value=0}
       ]}}}
   )
   &SpriteRotationRate(Rotation Rate=90)
   &GravityForce(Gravity=&MergeHelper_VectorSwitch(
     0=&RandomRangeVector(Minimum=<0 0 300>, Maximum=<0 0 1000>)
     1=&RandomRangeVector(Minimum=<0 0 400>, Maximum=<1 1 1200>)
     2=&RandomRangeVector(Minimum=<0 0 50>, Maximum=<0 0 250>)
     3=?
     Index=$Particles.PartitionParticles.Partition
     Num=::NewEnumerator1
    ))
   &CurlNoiseForce(
    Bias Noise=false
    Noise Frequency=20
    Noise Strength=&MergeHelper_FloatSwitch(0=5000, 1=5000, 2=0, 3=?, Index=$Particles.PartitionParticles.Partition, Num=::NewEnumerator1)
    Pan Noise=true
    Pan Noise Field=<0 0 1>
    Randomization Vector:Vector3f=<0.65 0.3 0.1>
   )
   &SolveForcesAndVelocity(Clamp Velocity=false, Limit Acceleration=false)
   &AlignSpriteToMeshOrientation(SetAlignment=true, SetFacing=true)
   &SubUVAnimation(
    Play Rate:NiagaraFloat=&MergeHelper_FloatSwitch(0=1, 1=1, 2=0, 3=?, Index=$Particles.PartitionParticles.Partition, Num=::NewEnumerator1)
    Random Start Frame=true
    "Sprite Renderer "=DataInterface() {propertyValues {SpriteRenderer=&self:Ash_0.NiagaraSpriteRendererProperties_1}}
    SubUV Animation Mode:ENiagara_SubUVLookupModeV2=::NewEnumerator3
    UseEndFrame=false
    UsePlayRateScale=false
    UseStartFrame=false
   ) {module=SubUVAnimation001}
  ]
  rendererClasses [&NiagaraSpriteRendererProperties]
  renderers [
   {rendererClass=&NiagaraSpriteRendererProperties}
   {rendererClass=&NiagaraSpriteRendererProperties, rendererIndex=1}
   {rendererClass=&NiagaraSpriteRendererProperties, rendererIndex=2}
  ]
 }
 Fountain {
  emitterUpdateScript [
   &EmitterState(
    Enable Distance Culling=true
    Life Cycle Mode=::NewEnumerator1
    Loop Duration=&RandomRangeFloat(Minimum=0.3, Maximum=0.7, Evaluation Type=::NewEnumerator1)
    MaxDistance=8000
    Module.bUseMaxDistance=true
    Module.bUseMinDistance=false
    Recalculate Duration Each Loop=true
    Scalability Mode=::NewEnumerator1
    UseLoopDelay=false
   )
   &SpawnBurst_Instantaneous(
    Loop Count Limit=-1
    Spawn Count=8
    Spawn Probability=&FN_PickFloatDependOnPlatform(TargetPlatformActive=0.22, TargetPlatformNoActive=0.14)
    Use Spawn Probability=true
    UseLoopCountLimit=false
   )
   &SpawnBurst_Instantaneous(Loop Count Limit=-1, Spawn Count=8, Spawn Group=1, Use Spawn Probability=false, UseLoopCountLimit=false) {module=SpawnBurst_Instantaneous001}
  ]
  particleSpawnScript [
   /SetVariables(Particles.VisibilityTag=$Emitter.SpawnGroup) {isSetParameters=true}
   &InitializeParticle(
    Lifetime=&MaskFloatBySpawnGroup(Masked Float=&RandomRangeFloat(Minimum=5.5, Maximum=6), Passthrough Float=1)
    Mass Mode=::NewEnumerator3
    Mesh Scale Mode=::NewEnumerator3
    Position Offset=&RandomRangeVector(Minimum=<-200 -200 0>, Maximum=<200 200 0>)
    Sprite Rotation Mode=::NewEnumerator2
    Sprite Size Max=<500 600>
    Sprite Size Min=<350 400>
    Sprite Size Mode=::NewEnumerator2
    UsePositionOffset=true
    Write Material Random=true
   )
   &ShapeLocation(
    Additional Quaternion Rotation=true
    Additional Yaw / Pitch / Roll=<20 0 0>
    MaskSpawnGroup=false
    Offset=&MaskVectorBySpawnGroup(Masked Vector=<0 0 0>, Passthrough Vector=<0 0 30>)
    Offset Mode=::NewEnumerator0
    Override Seed=false
    Ring / Disc Distribution Mode:ENiagaraSphereDistributionMode=::NewEnumerator1
    Ring Radius=&RandomRangeFloat(Minimum=3500, Maximum=3800)
    Rotation Angle Type:ENiagara_AngleInput=::NewEnumerator0
    Rotation Mode=::NewEnumerator3
    Shape Primitive=::NewEnumerator4
    U Position=&ReturnNormalizedExecIndex(Normalized Index Includes One=false)
    UseOwnerScale=false
   )
   &AddVelocity(
    Module.Use Falloff Distance=false
    Override Seed=false
    Velocity Mode=::NewEnumerator1
    Velocity Speed=&RandomRangeFloat(Minimum=-500, Maximum=-850)
   )
   &InitialMeshOrientation(Use Rotation Vector=false)
   /SetVariables(Particles.MeshIndex=&RandomRangeInt(Minimum=0, Maximum=20)) {isSetParameters=true}
  ]
  particleUpdateScript [
   &ParticleState()
   &ScaleSpriteSize(
    Non-Uniform Curve Sprite Scale=DataInterface() {propertyValues {
      XCurve {
       defaultValue=1
       keys [<at 0 0.25 RCIM_Cubic>, {arriveTangent=3.5553, interpMode=RCIM_Cubic, leaveTangent=3.5553, tangentMode=RCTM_User, time=1, value=2}]
      }
      YCurve {
       defaultValue=1
       keys [<at 0 0.1 RCIM_Cubic>, {arriveTangent=1.8677, interpMode=RCIM_Cubic, leaveTangent=1.8677, tangentMode=RCTM_User, time=0.9951, value=1.0522}]
      }
     }}
    Scale Sprite Size Mode=::NewEnumerator3
   )
   &ScaleMeshSize(Scale Factor=&VectorFromFloat(Value=&FloatFromCurve(FloatCurve=DataInterface() {propertyValues {Curve {keys [<at 0 0>,<at 0.1 1>,<at 0.9 1>,<at 1 0>]}}})))
   &UpdateMeshOrientation(Rotation Vector=<1 0.5 0.2>)
   &GravityForce(Gravity=&MakeVector(
     X=0
     Y=0
     Z=&MaskFloatBySpawnGroup(
      Masked Float=&FloatFromCurve(
       FloatCurve=DataInterface() {propertyValues {Curve {defaultValue=1, keys [<at 0 0.1 RCIM_Cubic>, {tangentMode=RCTM_User, time=0.5, value=1}]}}}
       Scale Curve=200
      )
      Passthrough Float=75
     )
    ))
   &Drag(Drag=0.25, Use Linear Drag=true, Use Rotational Drag=false)
   &ScaleColor(
    Scale Alpha=&FloatFromCurve(
     FloatCurve=DataInterface() {propertyValues {Curve {
        defaultValue=1.0113
        keys [
         {arriveTangent=0.2213, interpMode=RCIM_Cubic, leaveTangent=0.2213, tangentMode=RCTM_User, time=0, value=0}
         {arriveTangent=-0.5973, interpMode=RCIM_Cubic, leaveTangent=-0.5973, tangentMode=RCTM_User, time=0.3919, value=0.16}
         {tangentMode=RCTM_User, time=1, value=0}
        ]
       }}}
     Scale Curve=0.5
    )
    ScaleA=true
    ScaleRGB=true
   )
   &CurlNoiseForce(
    Bias Noise=false
    Noise Strength=&MaskFloatBySpawnGroup(Masked Float=1000, Passthrough Float=0)
    Pan Noise=false
    Randomization Vector:Vector3f=<0.65 0.3 0.1>
   )
   &SolveForcesAndVelocity(Clamp Velocity=false, Limit Acceleration=false)
   &ScaleSpriteSizeBySpeed(Max Scale Factor=<0.5 7>)
   /SetVariables(Particles.DMP0:Vector4f=&MakeVector4(
     X=&FloatFromCurve(FloatCurve=DataInterface() {propertyValues {Curve {
         defaultValue=0
         keys [
          <at 0 1 RCIM_Cubic>
          {arriveTangent=-0.7797, time=0.1873, value=0}
          {arriveTangent=1.2304, interpMode=RCIM_Cubic, leaveTangent=1.2304, time=0.45, value=0}
          {arriveTangent=2.8936, interpMode=RCIM_Cubic, leaveTangent=2.8936, tangentMode=RCTM_User, time=1, value=1}
         ]
        }}})
     Y=0
     Z=0
     W=0
    )) {isSetParameters=true}
  ]
  rendererClasses [&NiagaraSpriteRendererProperties, &NiagaraMeshRendererProperties]
  renderers [
   {rendererClass=&NiagaraSpriteRendererProperties}
   {rendererClass=&NiagaraMeshRendererProperties, rendererIndex=1}
   {rendererClass=&NiagaraMeshRendererProperties, rendererIndex=2}
  ]
 }
}
summary {
 systemName=NS_LW_Debris_suction
 emitters [
  {emitterName=Fountain, rendererClasses [&NiagaraSpriteRendererProperties, &NiagaraMeshRendererProperties]}
  {emitterName=Ash, simTarget=GPUComputeSim, rendererClasses [&NiagaraSpriteRendererProperties]}
 ]
 userVariables [
  User.Spawn {type=NiagaraFloat, defaultValue=1, description=""}
  User.Spawn Rate {type=NiagaraFloat, defaultValue=1, description=""}
  User.SpawnBoxRange {type=Vector3f, description="", defaultValue {struct=Vector3f, value=<1500 1500 50>}}
  User.Velocity {type=Vector3f, description="", defaultValue {struct=Vector3f, value=<0 -500 0>}}
 ]
}
