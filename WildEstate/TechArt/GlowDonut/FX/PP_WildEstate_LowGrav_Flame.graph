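# Post-process material that overlays animated flame noise on the screen through a vignette, apparently for a WildEstate low-gravity effect.
# A panned swirl flowmap distorts two scrolling samples of a greyscale flame texture, which are summed.
# A remapped vignette mask blends plain and tinted flames with the scene, and the two saturated results are multiplied.
# Master FX Control fades the effect in; only VignetteColor's red channel is used, so the tint is monochrome.
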Material /WildEstate/TechArt/GlowDonut/FX/PP_WildEstate_LowGrav_Flame
graph {
 properties {materialDomain=MD_PostProcess, pixelDepthOffsetMode=PDOM_Legacy, usedShadingModels=Default Lit}
 nodes [
  n1 = Time()
  n2 = VectorParameter() {parameterName=Panner, defaultValue=<rgba 0.05 0.2 -0.02 0.1>}
  n3 = Add(A=TextureCoordinate() {uTiling=0.45, vTiling=0.3}, B=Multiply(A=@n1, B=ComponentMask(None=@n2.RGB) {g=true, r=true}))
  n4 = TextureSample(UVs=@n3) {samplerType=SAMPLERTYPE_LinearColor, texture=/Game/VisualThreatManager/StormVisuals/LargeClouds/T_Swirl_Flowmap}
  n5 = Multiply(A=ComponentMask(None=Subtract(A=@n4.RGB) {constB=0.5}) {g=true, r=true}, B=ScalarParameter() {defaultValue=0.2, parameterName=Distortion})
  n6 = Constant2Vector() {g=0.5, r=0.5}
  n7 = Subtract(A=TextureCoordinate(), B=@n6)
  n8 = Multiply(A=@n1, B=ComponentMask(None=@n2.RGBA) {a=true, b=true})
  n9 = Add(A=@n5, B=Add(A=@n7, B=@n8))
  n10 = AppendVector(
   A=ScalarParameter() {defaultValue=1.6, group=Use Texture, parameterName=U Tiling, sortPriority=2}
   B=ScalarParameter() {defaultValue=1, group=Use Texture, parameterName=V Tiling, sortPriority=3}
  )
  n11 = TextureSampleParameter2D(UVs=Add(A=Multiply(A=@n9, B=@n10), B=@n6)) {
   group=Use Texture
   parameterName=Texture Mask
   samplerType=SAMPLERTYPE_LinearGrayscale
   sortPriority=1
   texture=/Game/Effects/Fort_Effects/Emotes/UltraLight/Texture/T_FluidFlameTile_Greyscale
  }
  n12 = Add(A=@n5, B=@n3)
  n13 = Add(A=@n7, B=Multiply(A=@n8) {constB=0.2})
  n14 = TextureSampleParameter2D(UVs=Multiply(A=Add(A=@n12, B=@n13), B=Constant2Vector() {g=0.8, r=0.9}) {constB=0.8}) {
   group=Use Texture
   parameterName=Texture Mask
   samplerType=SAMPLERTYPE_LinearGrayscale
   sortPriority=1
   texture=&T_FluidFlameTile_Greyscale
  }
  n15 = Add(A=@n11.RGBA, B=@n14.RGBA)
  n16 = ScalarParameter() {group=Vignette, parameterName=Vignette In Intensity, sliderMax=1, sortPriority=2}
  n17 = ScalarParameter() {defaultValue=0.666, group=Vignette, parameterName=Vignette Out Intensity, sliderMax=1, sortPriority=1}
  n18 = Saturate(None=MaterialFunctionCall(
    Input=MaterialFunctionCall() {materialFunction=/Game/Creative/PostProcess/MaterialFunctions/MF_VignetteMask}.VignetteMask
    Input Low=Constant()
    Input High=Constant() {r=1}
    Target Low=OneMinus(None=@n16)
    Target High=OneMinus(None=@n17)
   ) {materialFunction=&RemapValueRange}.Result)
  n19 = SceneTexture() {sceneTextureId=PPI_PostProcessInput0}
  n20 = LinearInterpolate(A=@n15, B=@n19.Color, Alpha=@n18) {constA=1}
  n21 = VectorParameter() {group=Vignette, parameterName=VignetteColor, sortPriority=0, defaultValue=<rgba 4 0 1.6678 1>}
  n22 = LinearInterpolate(
   A=Multiply(
    A=@n15
    B=MaterialFunctionCall(X=@n21.R, Y=@n21.R, Z=@n21.R, A=@n21.R) {materialFunction=&MakeFloat4}.Result
   )
   B=@n19.Color
   Alpha=@n18
  )
  LinearInterpolate(
   A=SceneTexture() {sceneTextureId=PPI_PostProcessInput0}.Color
   B=Multiply(A=Saturate(None=@n20), B=Saturate(None=@n22))
   Alpha=ScalarParameter() {defaultValue=0.1, group=Vignette, parameterName=Master FX Control, sortPriority=0}
  )
 ]
}
